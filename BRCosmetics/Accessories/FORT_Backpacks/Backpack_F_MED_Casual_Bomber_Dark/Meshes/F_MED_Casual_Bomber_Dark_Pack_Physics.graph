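# Physics asset for the Casual Bomber Dark backpack cosmetic, giving its wings, tail, chain, zipper, mouth and fire joints physical bodies and limits.
# Small capsules on each joint are constrained back to the pack body, with two-link wings and three-link tail and chain.
# Most joints have narrow swing and twist limits.
# The mouth and its child fire joint lock twist and one swing axis each, leaving a single limited hinge-like swing.
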PhysicsAsset /BRCosmetics/Accessories/FORT_Backpacks/Backpack_F_MED_Casual_Bomber_Dark/Meshes/F_MED_Casual_Bomber_Dark_Pack_Physics
bodies {
 dyn_L_wing_jnt_1 [{shapeType=Capsule, length=2, radius=1, center=<0.155 0 0>}]
 dyn_L_wing_jnt_2 [{shapeType=Capsule, length=4, radius=2, rotation=<rot 90 -180 -180>}]
 dyn_R_wing_jnt_1 [{shapeType=Capsule, length=2, radius=1, center=<0.155 0 0>}]
 dyn_R_wing_jnt_2 [{shapeType=Capsule, length=4, radius=2, rotation=<rot 90 -180 -180>}]
 dyn_chain_jnt_1 [{shapeType=Capsule, length=1, radius=0.5, center=<0.4939 0 -0.0378>, rotation=<rot -86.0344 179.9988 -179.9987>}]
 dyn_chain_jnt_2 [{shapeType=Capsule, length=1, radius=0.5, center=<0.4939 0 -0.0378>, rotation=<rot -86.0352 179.9986 -179.9987>}]
 dyn_chain_jnt_3 [{shapeType=Capsule, length=1, radius=0.5, center=<0.4939 0 -0.038>, rotation=<rot -86.0063 179.9986 -179.9986>}]
 dyn_fire_jnt [{shapeType=Capsule, length=5, radius=1, center=<2.6414 0 -0.0464>, rotation=<rot -88.9485 -180 180>}]
 dyn_mouth_jnt [{shapeType=Capsule, length=1, radius=1.7, center=<0.5 0 0>, rotation=<rot 0 0 90>}]
 dyn_tail_jnt_1 [{shapeType=Capsule, length=1, radius=1, center=<0.743 0 0.1454>, rotation=<rot 79 180 180>}]
 dyn_tail_jnt_2 [{shapeType=Capsule, length=1, radius=1, center=<0.743 0 0.1454>, rotation=<rot 79 180 180>}]
 dyn_tail_jnt_3 [{shapeType=Capsule, length=1, radius=1, center=<0.743 0 0.1454>, rotation=<rot 79 180 180>}]
 dyn_zipper_jnt [{shapeType=Capsule, length=2, radius=0.5, center=<1 0 0>, rotation=<rot -90 0 0>}]
 pack [{shapeType=Capsule, length=1, radius=1}]
}
constraints [
 {bone1Name=dyn_L_wing_jnt_1, bone2Name=pack, swing1LimitDegrees=25, swing2LimitDegrees=10, twistLimitDegrees=10}
 {bone1Name=dyn_L_wing_jnt_2, bone2Name=dyn_L_wing_jnt_1, swing1LimitDegrees=20, swing2LimitDegrees=5, twistLimitDegrees=10}
 {bone1Name=dyn_R_wing_jnt_1, bone2Name=pack, swing1LimitDegrees=25, swing2LimitDegrees=10, twistLimitDegrees=10}
 {bone1Name=dyn_R_wing_jnt_2, bone2Name=dyn_R_wing_jnt_1, swing1LimitDegrees=20, swing2LimitDegrees=5, twistLimitDegrees=10}
 {bone1Name=dyn_tail_jnt_1, bone2Name=pack, swing1LimitDegrees=5, swing2LimitDegrees=10, twistLimitDegrees=4}
 {bone1Name=dyn_tail_jnt_2, bone2Name=dyn_tail_jnt_1, swing1LimitDegrees=5, swing2LimitDegrees=10, twistLimitDegrees=4}
 {bone1Name=dyn_tail_jnt_3, bone2Name=dyn_tail_jnt_2, swing1LimitDegrees=5, swing2LimitDegrees=10, twistLimitDegrees=4}
 {
  bone1Name=dyn_mouth_jnt
  bone2Name=pack
  swing1LimitDegrees=45
  swing1Motion=Locked
  swing2LimitDegrees=50
  twistLimitDegrees=45
  twistMotion=Locked
 }
 {bone1Name=dyn_chain_jnt_1, bone2Name=pack, swing1LimitDegrees=20, swing2LimitDegrees=20, twistLimitDegrees=20}
 {bone1Name=dyn_chain_jnt_2, bone2Name=dyn_chain_jnt_1, swing1LimitDegrees=20, swing2LimitDegrees=20, twistLimitDegrees=20}
 {bone1Name=dyn_chain_jnt_3, bone2Name=dyn_chain_jnt_2, swing1LimitDegrees=20, swing2LimitDegrees=20, twistLimitDegrees=20}
 {bone1Name=dyn_zipper_jnt, bone2Name=pack, swing1LimitDegrees=15, swing2LimitDegrees=5, twistLimitDegrees=3}
 {
  bone1Name=dyn_fire_jnt
  bone2Name=dyn_mouth_jnt
  swing1LimitDegrees=7
  swing2LimitDegrees=45
  swing2Motion=Locked
  twistLimitDegrees=45
  twistMotion=Locked
 }
]
